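# Sound cue for the reactive sound of the Calavera back bling cosmetic.
# A fixed short delay feeds a Random node that picks one of three equally weighted reactive wave variants.
# It uses a close-range pet attenuation asset, a reactive backbling audio property sheet and the backblings sound class.
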SoundCue /BRCosmetics/Sounds/Backblings/Calavera/Backpack_Calavera_React_Cue
root=Delay_0
nodes [
 {id=Delay_0, children [Random_0], properties {delayMax=0.2, delayMin=0.2}}
 {id=Random_0, children [WavePlayer_0,WavePlayer_1,WavePlayer_5], properties {weights [1,1,1]}}
 {id=WavePlayer_0, properties {soundWaveAssetPtr=&/Backbling_Calavera_Reactive_01}}
 {id=WavePlayer_1, properties {soundWaveAssetPtr=&/Backbling_Calavera_Reactive_02}}
 {id=WavePlayer_5, properties {soundWaveAssetPtr=&/Backbling_Calavera_Reactive_03}}
]
properties {
 attenuationSettings=&Fort_Pets_Close_Athena_Att
 audioPropertiesSheet=/BRCosmetics/Sounds/Settings/PropertySheets/Reactive/APS_Cosmetic_Reactive_Backbling
 maxDistance=1800.0001
 soundClassObject=/Game/Sounds/FortSoundClassesAndModes/Fort_Backblings
}
